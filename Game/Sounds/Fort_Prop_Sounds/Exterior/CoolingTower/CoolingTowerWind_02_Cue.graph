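# Sound cue for looping cooling tower wind ambience, in the outdoor-ducking ambient sound class with a low voice limit.
# One CoolingTowerWind_01 wave player feeds two capsule attenuation branches that a mixer sums.
# The first is non-spatialized with logarithmic falloff.
# The second, mixed slightly lower, is spatialized with listener focus, distance filtering and a custom curve audible only in an outer band of its falloff.
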SoundCue /Game/Sounds/Fort_Prop_Sounds/Exterior/CoolingTower/CoolingTowerWind_02_Cue
root=Mixer_0
nodes [
 {
  id=Attenuation_1
  children [WavePlayer_1]
  properties {
   bOverrideAttenuation=true
   attenuationOverrides {
    attenuationShape=Capsule
    bSpatialize=false
    distanceAlgorithm=Logarithmic
    falloffDistance=2000
    reverbWetLevelMin=0.6429
    attenuationShapeExtents=<9126 1500>
   }
  }
 }
 {
  id=Attenuation_2
  children [WavePlayer_1]
  properties {
   bOverrideAttenuation=true
   attenuationOverrides {
    attenuationShape=Capsule
    bAttenuateWithLPF=true
    bEnableFocusInterpolation=true
    bEnableListenerFocus=true
    bEnableLogFrequencyScaling=true
    distanceAlgorithm=Custom
    falloffDistance=4000
    focusAttackInterpSpeed=0.1
    focusReleaseInterpSpeed=0.1
    hPFFrequencyAtMax=350
    lPFFrequencyAtMax=1000
    lPFRadiusMax=4000
    nonFocusVolumeAttenuation=0.4
    reverbWetLevelMin=0.6429
    stereoSpread=300
    attenuationShapeExtents {x=9126}
    customAttenuationCurve {editorCurveData {keys [
       <at 0.65 0 RCIM_Cubic>
       {arriveTangent=0.007, interpMode=RCIM_Cubic, leaveTangent=0.007, tangentMode=RCTM_User, time=0.75, value=0.8}
       {arriveTangent=-6.3675, interpMode=RCIM_Cubic, leaveTangent=-6.3675, tangentMode=RCTM_User, time=0.8174, value=0.6232}
       {interpMode=RCIM_Cubic, tangentMode=RCTM_User, time=1.1, value=0}
      ]}}
   }
  }
 }
 {id=Mixer_0, children [Attenuation_1, Attenuation_2], properties {inputVolume [1,0.8]}}
 {id=WavePlayer_1, properties {bLooping=true, soundWaveAssetPtr=&/CoolingTowerWind_01}}
]
properties {
 bDebug=true
 bOverrideConcurrency=true
 maxDistance=13526
 soundClassObject=/Game/Sounds/FortSoundClassesAndModes/Fort_Ambient_DuckOutdoors_Athena_PSM
 volumeMultiplier=1
 concurrencyOverrides {maxCount=2}
}
