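# Ambient sound cue for a power plant vat, filed with the cooling tower exterior props and routed to an indoor ambient sound class.
# A mixer layers two looping waves, a vat bed and a louder-weighted center loop, each behind its own attenuation override.
# The bed uses non-spatialized box falloff, while the center loop is spatialized with spherical falloff and stereo spread.
# Cue-level attenuation also defines a box with complex-collision occlusion and low-pass filtering.
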SoundCue /Game/Sounds/Fort_Prop_Sounds/Exterior/CoolingTower/PowerPlant_Vat_Cue
root=Mixer_0
nodes [
 {
  id=Attenuation_0
  children [WavePlayer_0]
  properties {
   bOverrideAttenuation=true
   attenuationOverrides {attenuationShape=Box, bSpatialize=false, falloffDistance=2000, attenuationShapeExtents {y=1000, z=1000}}
  }
 }
 {
  id=Attenuation_1
  children [WavePlayer_1]
  properties {bOverrideAttenuation=true, attenuationOverrides {falloffDistance=1322.7162, stereoSpread=400, attenuationShapeExtents {x=990.6151}}}
 }
 {id=Mixer_0, children [Attenuation_0, Attenuation_1], properties {inputVolume [1,3]}}
 {id=WavePlayer_0, properties {bLooping=true, soundWaveAssetPtr=&/PowerPlant_Vat_01}}
 {
  id=WavePlayer_1
  properties {bLooping=true, soundWaveAssetPtr=&/PowerPlant_VatCenter_01Loop}
 }
]
properties {
 bOverrideAttenuation=true
 maxDistance=3000
 soundClassObject=&Fort_Ambient_Indoors_Athena
 volumeMultiplier=0.125
 attenuationOverrides {
  attenuationShape=Box
  bEnableOcclusion=true
  bSpatialize=false
  bUseComplexCollisionForOcclusion=true
  falloffDistance=2000
  occlusionInterpolationTime=2
  occlusionLowPassFilterFrequency=1000
  occlusionTraceChannel=ECC_GameTraceChannel11
  attenuationShapeExtents {y=1000, z=1000}
 }
}
